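# Sound cue for the Hightower Date gauntlet's aim-down-sights-in foley sound.
# A Random node picks between two equally weighted ADS wave variants and feeds a root Modulator node for playback variation.
# It plays quietly in the weapon foley sound class, using the shared weapon reload attenuation.
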SoundCue /HighTower/Audio/Date/Hightower_Date_ADS_In_Cue
root=Modulator_0
nodes [
 {id=Modulator_0, children [Random_0]}
 {id=Random_0, children [WavePlayer_1,WavePlayer_2], properties {weights [1,1]}}
 {id=WavePlayer_1, properties {soundWaveAssetPtr=&/Hightower_Date_Gauntlet_ADS}}
 {id=WavePlayer_2, properties {soundWaveAssetPtr=&/Hightower_Date_Gauntlet_ADS_02}}
]
properties {
 attenuationSettings=&Fort_Weapon_Reload_Att
 maxDistance=2250
 soundClassObject=&Fort_WeaponFoley
 volumeMultiplier=0.25
}
