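# Sound cue for the HighTower "Date" ball lightning equip sound, judging by its name.
# A single wave player feeds the ball lightning start wave through a default Modulator at reduced cue volume.
# It uses close projectile attenuation and a first-person ranged weapon fire sound class.
# A concurrency override allows one instance at a time and blocks new plays while one is active.
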SoundCue /HighTower/Audio/Date/Hightower_Date_Ball_LIghtning_Equip_Cue
root=Modulator_0
nodes [
 {id=Modulator_0, children [WavePlayer_5]}
 {id=WavePlayer_5, properties {soundWaveAssetPtr=&/Hightower_Date_Ball_Lightning_Start_01}}
]
properties {
 attenuationSettings=&Fort_Weapon_Projectile_Close_Athena_Att
 bOverrideConcurrency=true
 maxDistance=4950
 soundClassObject=&Fort_WeaponRanged_Fire_1P_PSM
 volumeMultiplier=0.4
 concurrencyOverrides {maxCount=1, resolutionRule=PreventNew}
}
