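# Sound cue for the looping audio of the Hightower Date ball lightning, attenuated as a close weapon projectile.
# A looping wave feeds an enveloper that fades volume in from near silence over about 1.5 seconds.
# That passes through attenuation into a root sound class node that overrides to the first-person ranged weapon fire class.
# Overridden concurrency caps it at two instances per owner, stopping the oldest.
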SoundCue /HighTower/Audio/Date/Hightower_Date_Ball_Lightning_Loop_Cue
root=SoundClass_5
nodes [
 {id=Attenuation_0, children [Enveloper_1], properties {attenuationSettings=&Fort_Weapon_Projectile_Close_Athena_Att}}
 {
  id=Enveloper_1
  children [WavePlayer_5]
  properties {
   pitchCurve {editorCurveData {keys [{time=0}]}}
   volumeCurve {editorCurveData {keys [<at 0 0.001>,<at 0.67 0.037>,<at 1.2111 0.3828>,<at 1.4627 1.0032>]}}
  }
 }
 {
  id=SoundClass_5
  children [Attenuation_0]
  properties {soundClassOverride=&Fort_WeaponRanged_Fire_1P_PSM}
 }
 {id=WavePlayer_5, properties {bLooping=true, soundWaveAssetPtr=&/Hightower_Date_Ball_Lighting_Loop_v5}}
]
properties {
 bDebug=true
 bOverrideConcurrency=true
 maxDistance=4950
 volumeMultiplier=0.35
 concurrencyOverrides {bLimitToOwner=true, maxCount=2, resolutionRule=StopOldest}
 concurrencySet [null]
}
