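# Sound cue that plays the Ball Lightning pickup sound in the HighTower Date audio set.
# A single wave player feeds a Modulator at default settings, which varies pitch and volume per play.
# The cue uses the shared weapon pickup attenuation and loot collect sound class at reduced volume.
# Its concurrency override prevents new instances once the limit is reached, rather than stopping playing ones.
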SoundCue /HighTower/Audio/Date/Hightower_Date_Ball_Lightning_Pickup_Cue
root=Modulator_0
nodes [
 {id=Modulator_0, children [WavePlayer_0]}
 {id=WavePlayer_0, properties {soundWaveAssetPtr=&/Hightower_Date_Ball_Lightnig_Pickup}}
]
properties {
 attenuationSettings=&Fort_Weapon_Pickup_Att
 bOverrideConcurrency=true
 maxDistance=3500
 soundClassObject=&Fort_Loot_Collect
 volumeMultiplier=0.2
 concurrencyOverrides {resolutionRule=PreventNew}
}
