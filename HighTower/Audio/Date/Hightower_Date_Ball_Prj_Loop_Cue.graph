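# Sound cue for the looping projectile sound of the HighTower "Date" ball, judging by its name.
# A looping wave player reuses the Date aura loop wave and feeds a modulator that adds very slight random pitch variation.
# The cue plays it pitched down at half volume through the Athena player foley sound class.
# An attenuation override on the cue applies logarithmic distance falloff with stereo spread.
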SoundCue /HighTower/Audio/Date/Hightower_Date_Ball_Prj_Loop_Cue
root=Modulator_0
nodes [
 {id=Modulator_0, children [WavePlayer_1], properties {pitchMax=1.01, pitchMin=0.99}}
 {id=WavePlayer_1, properties {bLooping=true, soundWaveAssetPtr=&/Hightower_Date_Aura_Loop_v2}}
]
properties {
 bOverrideAttenuation=true
 maxDistance=2512
 pitchMultiplier=0.7
 priority=0.5
 soundClassObject=&Fort_Player_Foley_Athena
 volumeMultiplier=0.5
 attenuationOverrides {distanceAlgorithm=Logarithmic, falloffDistance=2000, stereoSpread=512, attenuationShapeExtents {x=512}}
}
