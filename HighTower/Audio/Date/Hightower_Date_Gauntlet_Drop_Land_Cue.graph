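# Sound cue for the landing sound when the Hightower Date gauntlet is dropped.
# A single wave player feeds the gauntlet drop-land wave through a modulator at the root, with boosted cue volume.
# It uses the shared Fortnite weapon-drop attenuation and routes through the loot-collect sound class.
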SoundCue /HighTower/Audio/Date/Hightower_Date_Gauntlet_Drop_Land_Cue
root=Modulator_0
nodes [
 {id=Modulator_0, children [WavePlayer_0]}
 {id=WavePlayer_0, properties {soundWaveAssetPtr=/HighTower/Audio/Date/S31_Combo/Date_Gauntlet_Drop_Land_01}}
]
properties {
 attenuationSettings=&Fort_Weapon_Drop_Att
 maxDistance=3000
 soundClassObject=&Fort_Loot_Collect
 volumeMultiplier=1.5
}
