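# Sound cue for the gauntlet equip foley of the HighTower Date character or item.
# A single wave player plays one equip sample with no randomization or mixing.
# It routes to the weapon foley sound class and uses the weapon reload attenuation asset.
# It overrides concurrency with a prevent-new rule, so extra instances do not start while the limit is reached.
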SoundCue /HighTower/Audio/Date/Hightower_Date_Gauntlet_Equip_Cue
root=WavePlayer_0
nodes [{id=WavePlayer_0, properties {soundWaveAssetPtr=/HighTower/Audio/Date/S31_Combo/Date_Gauntlet_Equip_01}}]
properties {
 attenuationSettings=&Fort_Weapon_Reload_Att
 bOverrideConcurrency=true
 maxDistance=2250
 soundClassObject=&Fort_WeaponFoley
 volumeMultiplier=1
 attenuationOverrides {falloffDistance=4100, attenuationShapeExtents {x=900}}
 concurrencyOverrides {resolutionRule=PreventNew}
}
