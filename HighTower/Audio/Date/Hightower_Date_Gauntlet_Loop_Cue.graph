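# Sound cue for the Hightower Date gauntlet, pairing a looping aura with a one-shot equip sound.
# The loop fades in over half a second and mixes with the quieter, pitched-down equip sound, then a shared envelope lowers both over ten seconds.
# IsLocal and IsTeammate switches route this shared chain to an unattenuated local mix, a friendly-attenuated mix or a louder default-attenuated mix for others.
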SoundCue /HighTower/Audio/Date/Hightower_Date_Gauntlet_Loop_Cue
root=Switch_2
nodes [
 {id=Attenuation_0, children [Mixer_4], properties {attenuationSettings=/HighTower/Audio/Aura/Default/Hightower_Aura_Friendly_Att}}
 {id=Attenuation_3, children [Mixer_8], properties {attenuationSettings=/HighTower/Audio/Aura/Default/Hightower_Aura_Att}}
 {
  id=Enveloper_2
  children [WavePlayer_0]
  properties {pitchCurve {editorCurveData {keys [{time=0}]}}, volumeCurve {editorCurveData {keys [<at 0.0117 0.0072>,<at 0.5073 1>]}}}
 }
 {
  id=Enveloper_5
  children [Mixer_0]
  properties {pitchCurve {editorCurveData {keys [{time=0}]}}, volumeCurve {editorCurveData {keys [<at 0 1>,<at 3.1792 0.9391>,<at 10 0.35>]}}}
 }
 {id=Mixer_0, children [Enveloper_2,Modulator_0], properties {inputVolume [1,0.4]}}
 {id=Mixer_4, children [Enveloper_5], properties {inputVolume [0.5]}}
 {id=Mixer_6, children [Enveloper_5], properties {inputVolume [0.5]}}
 {id=Mixer_8, children [Enveloper_5], properties {inputVolume [0.7]}}
 {id=Modulator_0, children [WavePlayer_2], properties {pitchMax=0.7, pitchMin=0.7, volumeMax=1, volumeMin=1}}
 {id=Switch_2, children [Switch_3,Switch_3,Mixer_6], properties {intParameterName=IsLocal}}
 {id=Switch_3, children [Attenuation_3, Attenuation_3, Attenuation_0], properties {intParameterName=IsTeammate}}
 {id=WavePlayer_0, properties {bLooping=true, soundWaveAssetPtr=&/Hightower_Date_Aura_Loop_v6}}
 {id=WavePlayer_2, properties {soundWaveAssetPtr=&/Hightower_Date_Gauntlet_Equip}}
]
properties {
 maxDistance=650
 priority=0.5
 soundClassObject=&Fort_Player_Foley_Athena
 volumeMultiplier=1.2
 attenuationOverrides {falloffDistance=2048, stereoSpread=512, attenuationShapeExtents {x=512}}
}
