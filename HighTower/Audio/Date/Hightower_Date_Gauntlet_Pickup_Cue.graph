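# Sound cue for the Hightower Date gauntlet pickup, by its name and its weapon pickup and loot collect settings.
# A single wave player feeds a Modulator node at its default settings, with reduced cue volume.
# It uses the shared weapon pickup attenuation and the loot collect sound class.
# A concurrency override prevents new instances from starting once the limit is reached, rather than stopping playing ones.
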SoundCue /HighTower/Audio/Date/Hightower_Date_Gauntlet_Pickup_Cue
root=Modulator_0
nodes [
 {id=Modulator_0, children [WavePlayer_0]}
 {id=WavePlayer_0, properties {soundWaveAssetPtr=/HighTower/Audio/Date/S31_Combo/Date_Gauntlet_Pickup_01}}
]
properties {
 attenuationSettings=&Fort_Weapon_Pickup_Att
 bOverrideConcurrency=true
 maxDistance=3500
 soundClassObject=&Fort_Loot_Collect
 volumeMultiplier=0.7
 concurrencyOverrides {resolutionRule=PreventNew}
}
